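# Niagara system for the Unicorn vehicle drift trail cosmetic, combining two trailing ribbons, spinning sprites and wheel-wrap meshes at the rear wheels.
# Sprites and meshes follow partition-indexed skeletal mesh sockets, ribbons and sprites inherit vehicle movement, and FN_DisableTireFX masks effects per wheel.
# Painted-variant overrides color every emitter, with sprites defaulting to random rainbow hues.
# Painted state also desaturates ribbon and wrap textures, while frontend preview lengthens sprite lifetimes.
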NiagaraSystem /VehicleCosmetics/DriftTrail/DriftTrail_Unicorn/FX/NS_DriftTrail_Unicorn_MP
dependencies {
 usedDataInterfaces [
  &NiagaraDataInterfaceCurve
  &NiagaraDataInterfaceSpriteRendererInfo
  &NiagaraDataInterfaceSkeletalMesh
 ]
 usedDynamicInputs []
 usedModules [
  &self:SystemSpawnScript.SystemScriptSource.SystemScriptGraph.NiagaraNodeAssignment_0.SetVariables_#1
  &FN_UnpackVehicleFlags
  /VehicleCosmetics/FX/Niagara/Modules/VC_SetVehicleSpeedAttribs
  &SystemState
 ]
}
emitters {
 NE_VC_Ribbon001 {
  emitterUpdateScript [
   &EmitterState(Life Cycle Mode=::NewEnumerator1, Loop Duration=1, UseLoopDelay=false)
   &VC_RateQualityHandler(High Spawn Rate=60, Low Spawn Rate=60, Medium Spawn Rate=60, Use Custom Rate?=true)
   &VC_SpawnRibbon_v1(Number of Ribbons=2, SpawnRate=$Emitter.Spawn Rate)
   /SetVariables(Emitter.TextureDesaturation:NiagaraFloat=&MakeCustomFloatFromBool(Bool=$User.bIsPainted, True Float=1, False Float=0)) {isSetParameters=true}
  ]
  particleSpawnScript [
   &VC_PaintedVariantsOverride(
    Black=<rgba 0.032 0.032 0.032 10>
    BurntSienna=<rgba 0.225 0.075 0 5>
    Cobalt=<rgba 0 0.127 1 10>
    Crimson=<rgba 0.4635 0.0309 0 5>
    ForestGreen=<rgba 0.05 0.3 0 5>
    Grey=<rgba 0.2062 0.2062 0.22 5>
    Lime=<rgba 0.3333 0.8 0 5>
    Orange=<rgba 0.8 0.32 0 10>
    Pink=<rgba 0.675 0.2194 0.5991 10>
    Purple=<rgba 0.1667 0 0.5 10>
    Saffron=<rgba 0.7792 0.85 0 10>
    SkyBlue=<rgba 0 0.4958 0.85 5>
    TitaniumWhite=<rgba 2 2 2 5>
   )
   &VC_InitializeRibbon_v1(
    Color=$Particles.Color
    Custom Offset=true
    Enable Skeletal Position=true
    Lifetime=0.65
    Local Vector Offset=&AddVectorToPosition(Position=$User.offset, Vector=<-12 0 -30>)
    Offset Amount=1
    Ribbon Width=45
   )
   &AddVelocity(Velocity=<-1400 0 0>) {module=AddVelocity001}
  ]
  particleUpdateScript [
   &FN_DisableTireFX(Mask Index Attribute:NiagaraInt32=$Particles.SpawnGroup, Mask Index FL=2, Mask Index FR=3, Mask Index RL=0, Mask Index RR=1)
   &ParticleState()
   &InheritSourceMovement(
    Applied Position Scale=&Multiply_VectorByFloat(
     Vector=<1 1 1>
     Float=&FloatFromCurve(FloatCurve=DataInterface() {propertyValues {Curve {keys [
          {interpMode=RCIM_Cubic, leaveTangent=-0.1311, tangentMode=RCTM_Break, time=0, value=1}
          {arriveTangent=-0.0803, interpMode=RCIM_Cubic, tangentMode=RCTM_Break, time=1, value=0.9}
         ]}}})
    )
    ApplyForce=false
    ApplyPosition=true
    ApplyPositionFalloff=false
   )
   &SolveForcesAndVelocity(Clamp Velocity=false, Limit Acceleration=false)
   &ScaleRibbonWidth(Ribbon Width Scale=&FloatFromCurve(FloatCurve=DataInterface() {propertyValues {Curve {
        defaultValue=0.4
        keys [
         {arriveTangent=3.754, interpMode=RCIM_Cubic, leaveTangent=3.754, tangentMode=RCTM_User, time=0.05, value=0.1}
         {arriveTangent=2.4679, interpMode=RCIM_Cubic, leaveTangent=2.4679, tangentMode=RCTM_User, time=0.2112, value=0.6467}
         {
          arriveTangent=0.5577
          arriveTangentWeight=0.0898
          interpMode=RCIM_Cubic
          leaveTangent=0.0233
          leaveTangentWeight=0.3368
          tangentMode=RCTM_Break
          tangentWeightMode=RCTWM_WeightedBoth
          time=0.4
          value=1
         }
         {arriveTangent=-1.9739, arriveTangentWeight=0.1529, interpMode=RCIM_Cubic, leaveTangent=-1.9739, tangentMode=RCTM_User, time=0.95, value=0}
        ]
       }}}))
   &VC_UpdateRibbon_v1(Enable Skeletal Position=true)
  ]
  rendererClasses [&NiagaraRibbonRendererProperties]
  renderers [{rendererClass=&NiagaraRibbonRendererProperties}]
 }
 NE_VC_Sprites {
  emitterUpdateScript [
   &EmitterState(Life Cycle Mode=::NewEnumerator1, Loop Behavior=::NewEnumerator1, Loop Duration Mode=::NewEnumerator1, UseLoopDelay=false)
   &SpawnRate(SpawnRate=77, Use Spawn Probability=false)
   &SpawnRate(SpawnGroup=1, SpawnRate=95, Use Spawn Probability=false) {module=SpawnRate001}
  ]
  particleSpawnScript [
   &VC_PaintedVariantsOverride(
    Black=<rgba 0.005 0.005 0.005 2>
    BurntSienna=<rgba 0.25 0.0833 0 1>
    Crimson=<rgba 0.2 0.0133 0 2>
    DefaultColor=/Niagara/DynamicInputs/LinearColor/MakeLinearColorFromHSV(
     HSV From:NiagaraInt32=1
     HSV:Vector3f=?
     Hue:NiagaraFloat=&RandomRangeFloat(Minimum=0, Maximum=1)
     Saturation:NiagaraFloat=0.885
     Value:NiagaraFloat=&RandomRangeFloat(Minimum=0.75, Maximum=1.45)
     Alpha:NiagaraFloat=1.25
    )
    ForestGreen=<rgba 0 0.18 0 2>
    Grey=<rgba 0.09 0.09 0.09 2>
    Lime=<rgba 0.0833 0.25 0 1>
    Orange=<rgba 0.45 0.225 0 2>
    Pink=<rgba 0.5052 0.0605 0.4311 1>
    Purple=<rgba 0.1333 0 0.4 1>
    R1 Black:LinearColor=<rgba 1 1 1 1>
    R1 BurntSienna:LinearColor=<rgba 6 1.9 0 1>
    R1 Cobalt:LinearColor=<rgba 0 0.8325 45 1>
    R1 Crimson:LinearColor=<rgba 10 0 0.0833 1>
    R1 DefaultColor=&MakeLinearColorFromHSV(
     HSV From:NiagaraInt32=1
     HSV:Vector3f=?
     Hue:NiagaraFloat=&RandomRangeFloat(Minimum=0, Maximum=1)
     Saturation:NiagaraFloat=0.75
     Value:NiagaraFloat=&RandomRangeFloat(Minimum=4, Maximum=6)
     Alpha:NiagaraFloat=1.25
    )
    R1 ForestGreen:LinearColor=<rgba 1 6 0 1>
    R1 Grey:LinearColor=<rgba 3 3 3 1>
    R1 Lime:LinearColor=<rgba 3.6667 5.5 0 1>
    R1 Orange:LinearColor=<rgba 6 4.5 0 1>
    R1 Pink:LinearColor=<rgba 15 3.75 10.9091 1>
    R1 Purple:LinearColor=<rgba 3.2833 0 8 1>
    R1 Saffron:LinearColor=<rgba 7.3333 8 0 1>
    R1 SkyBlue:LinearColor=<rgba 0 6.6539 10 1>
    R1 TitaniumWhite:LinearColor=<rgba 8 8 8 1>
    Saffron=<rgba 0.65 0.4983 0 1>
    SkyBlue=<rgba 0 0.2333 0.4 2>
    TitaniumWhite=<rgba 0.4238 0.4242 0.45 2>
    Use R1 Painted Overrides=true
   )
   &InitializeParticle(
    Color=$Particles.Color
    Lifetime Max=&MakeCustomFloatFromBool(
     Bool=$User.bIsFrontendPreview
     True Float=&MaskFloatBySpawnGroup(Masked Float=0.75, Passthrough Float=0.85)
     False Float=0.65
    )
    Lifetime Min:NiagaraFloat=&MakeCustomFloatFromBool(Bool=$User.bIsFrontendPreview, True Float=0.5, False Float=0.35)
    Lifetime Mode=::NewEnumerator1
    Sprite Rotation Mode=::NewEnumerator2
    Sprite Size Mode=::NewEnumerator1
    Uniform Sprite Size Max=&MaskFloatBySpawnGroup(Masked Float=20, Passthrough Float=7.5)
    Uniform Sprite Size Min=&MaskFloatBySpawnGroup(Masked Float=8, Passthrough Float=4.5)
    UsePositionOffset=false
    Write Material Random=true
   )
   &PartitionParticles(Particles Per Partition=4, Partition Mode:ENiagara_PartitionMode=::NewEnumerator1)
   &SkeletalMeshLocation(
    Bone / Socket Index=&Add_Integer(A=$Particles.PartitionParticles.Partition, B=4)
    Bone / Socket Sampling Mode=::NewEnumerator5
    Interpolate Mesh Sampling=false
    MaskSpawnGroup=false
    Mesh Sampling Type=::NewEnumerator6
    Sampled Position Offset=&AddVectorToPosition(Position=$User.offset, Vector=<-75 0 -20>)
    Skeletal Mesh=$System.SkeletalMesh
   )
   &AddVelocity(Velocity=&RandomRangeVector(Minimum=<-450 -12 0>, Maximum=<-450 12 12>))
   &ShapeLocation(
    MaskSpawnGroup=false
    Override Seed=false
    Sphere Radius=&MaskFloatBySpawnGroup(Masked Float=20, Passthrough Float=25)
    UseOwnerScale=true
   )
   /SetVariables(Particles.RenderVisTag:NiagaraInt32=&SpawnGroup()) {isSetParameters=true}
  ]
  particleUpdateScript [
   &FN_DisableTireFX(
    Mask Index Attribute:NiagaraInt32=$Particles.PartitionParticles.Partition
    Mask Index FL=2
    Mask Index FR=3
    Mask Index RL=0
    Mask Index RR=1
   )
   &ParticleState()
   &InheritSourceMovement(
    Applied Position Scale=&Multiply_VectorByFloat(
     Vector=<1 1 1>
     Float=&FloatFromCurve(FloatCurve=DataInterface() {propertyValues {Curve {keys [
          {interpMode=RCIM_Cubic, leaveTangent=-0.1214, tangentMode=RCTM_Break, time=0, value=1}
          {arriveTangent=0.0008, interpMode=RCIM_Cubic, tangentMode=RCTM_Break, time=1, value=0.95}
         ]}}})
    )
    ApplyForce=false
    ApplyPosition=true
    ApplyPositionFalloff=false
   )
   &ScaleSpriteSize(
    Uniform Curve Scale=2.5
    Uniform Curve Sprite Scale=DataInterface() {propertyValues {Curve {keys [
        {interpMode=RCIM_Cubic, tangentMode=RCTM_Break, time=0.5, value=1}
        {arriveTangent=-3.0414, interpMode=RCIM_Cubic, tangentMode=RCTM_Break, time=1, value=0.2}
       ]}}}
   )
   &SpriteRotationRate(Rotation Rate=600)
   &GravityForce(Gravity=<0 0 -30>)
   &SolveForcesAndVelocity(Clamp Velocity=false, Limit Acceleration=false)
   &SubUVAnimation(
    Random Start Frame=true
    "Sprite Renderer "=DataInterface() {propertyValues {SpriteRenderer=&self:NE_VC_Sprites_0.NiagaraSpriteRendererProperties_3}}
    SubUV Lookup Index Scale=0
    UseEndFrame=false
    UseStartFrame=false
   )
  ]
  rendererClasses [&NiagaraSpriteRendererProperties]
  renderers [
   {rendererClass=&NiagaraSpriteRendererProperties}
   {rendererClass=&NiagaraSpriteRendererProperties, rendererIndex=1}
  ]
 }
 NE_VC_WheelWrap {
  emitterUpdateScript [
   &EmitterState(
    Inactive Response=::NewEnumerator1
    Life Cycle Mode=::NewEnumerator1
    Loop Behavior=::NewEnumerator1
    Loop Duration Mode=::NewEnumerator1
    UseLoopDelay=false
   )
   &SpawnBurst_Instantaneous(Spawn Count=2, Use Spawn Probability=false, UseLoopCountLimit=false)
   /SetVariables(Emitter.TextureDesaturation:NiagaraFloat=&MakeCustomFloatFromBool(Bool=$User.bIsPainted, True Float=1, False Float=0)) {isSetParameters=true}
  ]
  particleSpawnScript [
   &VC_PaintedVariantsOverride(
    Black=<rgba 0.032 0.032 0.032 10>
    BurntSienna=<rgba 0.225 0.075 0 5>
    Cobalt=<rgba 0 0.1524 1.2 1>
    Crimson=<rgba 0.4635 0.0309 0 5>
    ForestGreen=<rgba 0.05 0.3 0 5>
    Grey=<rgba 0.2062 0.2062 0.22 5>
    Lime=<rgba 0.3333 0.8 0 5>
    Orange=<rgba 0.8 0.32 0 10>
    Pink=<rgba 0.675 0.2194 0.5991 10>
    Purple=<rgba 0.1667 0 0.5 10>
    Saffron=<rgba 0.7792 0.85 0 10>
    SkyBlue=<rgba 0 0.4958 0.85 5>
    TitaniumWhite=<rgba 2 2 2 5>
   )
   &InitializeParticle(
    Color Mode=::NewEnumerator0
    Lifetime=5
    Mesh Scale Mode=::NewEnumerator4
    Sprite Rotation Mode=::NewEnumerator2
    Sprite Size Mode=::NewEnumerator3
    UsePositionOffset=false
    Write Material Random=true
   )
   &PartitionParticles()
   &SkeletalMeshLocation(
    Bone / Socket Index=&Add_Integer(A=$Particles.PartitionParticles.Partition, B=4)
    Bone / Socket Sampling Mode=::NewEnumerator5
    Interpolate Mesh Sampling=false
    MaskSpawnGroup=false
    Mesh Sampling Type=::NewEnumerator6
    Skeletal Mesh=$System.SkeletalMesh
   )
   &InitialMeshOrientation(Mesh Orientation Mode=::NewEnumerator7, Rotation=<0 0 0.5>, Use Rotation Vector=true)
   &Color(Color=<rgba 0 1 0 1>) {enabled=false}
  ]
  particleUpdateScript [
   &FN_DisableTireFX(
    Mask Index Attribute:NiagaraInt32=$Particles.PartitionParticles.Partition
    Mask Index FL=2
    Mask Index FR=3
    Mask Index RL=0
    Mask Index RR=1
   )
   &VC_WheelWrapScale(Wheel Index Attribute:NiagaraInt32=&Add_Integer(A=$Particles.PartitionParticles.Partition, B=2))
   &ParticleState(Kill Particles When Lifetime Has Elapsed=false, Loop Particles Lifetime=true)
   &UpdateMeshOrientation(Rotation Coordinate Space=::NewEnumerator2, Rotation Vector=<0 1 0>) {enabled=false}
   &SkeletalMeshLocation(
    Bone / Socket Index=&Add_Integer(A=$Particles.PartitionParticles.Partition, B=4)
    Bone / Socket Sampling Mode=::NewEnumerator5
    MaskSpawnGroup=false
    Mesh Sampling Type=::NewEnumerator6
    Skeletal Mesh=$System.SkeletalMesh
   ) {module=SkeletalMeshLocation001}
   &SolveForcesAndVelocity(Clamp Velocity=false, Limit Acceleration=false)
   &ScaleColor(ScaleA=true, ScaleRGB=true)
  ]
  rendererClasses [&NiagaraMeshRendererProperties]
  renderers [{rendererClass=&NiagaraMeshRendererProperties, rendererIndex=0}]
 }
}
summary {
 systemName=NS_DriftTrail_Unicorn_MP
 emitters [
  {emitterName=NE_VC_Sprites, rendererClasses [&NiagaraSpriteRendererProperties]}
  {emitterName=NE_VC_WheelWrap, rendererClasses [&NiagaraMeshRendererProperties]}
  {emitterName=NE_VC_Ribbon001, rendererClasses [&NiagaraRibbonRendererProperties]}
 ]
 userVariables [
  User.bIsFrontendPreview {type=NiagaraBool}
  User.bIsPainted {type=NiagaraBool}
  User.ColorVariantID {type=NiagaraInt32}
  User.offset {type=NiagaraPosition, description="", defaultValue {}}
  User.PaintedColor {type=LinearColor, description="", defaultValue {struct=LinearColor, value=<rgba 1 1 1 1>}}
  User.PaintedScalar {type=NiagaraFloat, defaultValue=1, description=""}
  User.Shine {type=NiagaraFloat, defaultValue=5, description=""}
  User.SpecMult {type=NiagaraFloat, defaultValue=0, description=""}
  User.SpecPower {type=NiagaraFloat, defaultValue=0, description=""}
  User.VehicleFlags {type=NiagaraInt32, defaultValue=806289408}
  User.Wheel_Radius_Array {
   type=NiagaraDataInterfaceArrayFloat
   defaultValue {
    struct=NiagaraExt_VariableValue_DataInterface
    value {
     dataInterface=&self:NiagaraDataInterfaceArrayFloat_1
     dataInterfaceClass=&NiagaraDataInterfaceArrayFloat
    }
   }
  }
  User.Wheel_Width_Array {
   type=NiagaraDataInterfaceArrayFloat
   defaultValue {
    struct=NiagaraExt_VariableValue_DataInterface
    value {
     dataInterface=&self:NiagaraDataInterfaceArrayFloat_0
     dataInterfaceClass=&NiagaraDataInterfaceArrayFloat
    }
   }
  }
  User.WheelTouchingArray {
   type=NiagaraDataInterfaceArrayFloat
   defaultValue {
    struct=NiagaraExt_VariableValue_DataInterface
    value {
     dataInterface=&self:NiagaraDataInterfaceArrayFloat_2
     dataInterfaceClass=&NiagaraDataInterfaceArrayFloat
    }
   }
  }
 ]
}
